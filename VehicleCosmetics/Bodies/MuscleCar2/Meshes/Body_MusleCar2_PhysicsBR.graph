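# Physics asset for the MuscleCar2 vehicle body mesh in the vehicle cosmetics content, named as a BR variant.
# It holds a single body on the root bone, with no shapes listed for it.
# It defines no constraints, so there are no joints between parts.
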PhysicsAsset /VehicleCosmetics/Bodies/MuscleCar2/Meshes/Body_MusleCar2_PhysicsBR
bodies {root []}
constraints []
