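# Physics asset that gives the MuscleCar2 vehicle body skeletal mesh a simple hitbox.
# It holds a single body on the root bone, made of one box shape covering roughly a car-sized volume, shifted forward along X and raised above the bone origin.
# It has no constraints, so the whole body acts as one collision volume with no articulated parts.
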PhysicsAsset /VehicleCosmetics/Bodies/MuscleCar2/Meshes/SK_Body_MusleCar2_HitBox
bodies {root [{shapeType=Box, extentX=409.3657, extentY=266.4958, extentZ=100.16, shapeName=root_box, center=<28.8 0 50.4>}]}
constraints []
